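; Anim notify state that adjusts the playback rate of a player's current Mantis technique montage, named for melee recovery.
; On begin, it casts the mesh owner to a FortPlayerPawn and reads a configurable float attribute.
; If that value is positive and not 1.0, it becomes the montage's play rate, using the montage from the pawn's FortMantisPawnComponent.
; The end event only returns true and does not restore the rate.
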(fn Received_NotifyBegin
  (params (MeshComp :type Skeletal%20Mesh%20Component%20Object%20Reference) (Animation :type Anim%20Sequence%20Base%20Object%20Reference) (TotalDuration :type Float%20%28single-precision%29) (EventReference :type Anim%20Notify%20Event%20Reference%20Structure%20%28by%20ref%29))
  (bind _attribute (Variables|Default|GetAttribute))
  (bind _asfort_player_pawn (Utilities|Casting|CastToFortPlayerPawn (Components|GetOwner MeshComp))
    (:then
      (LocalVariables|SetThisFortPawn _asfort_player_pawn)
      (LocalVariables|SetThisAnimInstance (Components|SkeletalMesh|GetAnimInstance MeshComp))
      (if (and (!= (.ReturnValue (Ability|Attribute|GetFloatAttribute _asfort_player_pawn _attribute)) 1.0) (> (.ReturnValue (Ability|Attribute|GetFloatAttribute _asfort_player_pawn _attribute)) 0))
        (Animation|Montage|MontageSetPlayRate (LocalVariables|GetThisAnimInstance) (Mantis|GetMontageforCurrentTechnique (Actor|GetComponentbyClass (LocalVariables|GetThisFortPawn) "/Script/MantisRuntime.FortMantisPawnComponent")) (.ReturnValue (Ability|Attribute|GetFloatAttribute _asfort_player_pawn _attribute)))
        (return true)
        (else
          (return false))))
    (:CastFailed
      (return false))))

(fn Received_NotifyEnd
  (params (MeshComp :type Skeletal%20Mesh%20Component%20Object%20Reference) (Animation :type Anim%20Sequence%20Base%20Object%20Reference) (EventReference :type Anim%20Notify%20Event%20Reference%20Structure%20%28by%20ref%29))
  (return true))
